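# Niagara impact effect for the MedievalSheath emote: a one-shot burst of spikes, rock debris and smoke.
# Impact_Spikes throws short-lived SubUV sprites outward, while two mesh rock emitters toss randomly chosen, tumbling debris that falls under gravity, collides and shrinks away.
# Smoke_Center and Smoke_Radial add expanding ring-spawned puffs animated through dynamic material parameters.
# Spike and radial smoke opacity scales with time of day.
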NiagaraSystem /BRCosmetics/Animation/Game/MainPlayer/Emotes/MedievalSheath/FX/NS_Emote_MedievalSheath_Impact
dependencies {
 usedDataInterfaces [&NiagaraDataInterfaceCurve, &NiagaraDataInterfaceColorCurve, &NiagaraDataInterfaceVector2DCurve]
 usedDynamicInputs []
 usedModules [&SystemState]
}
emitters {
 Impact_Spikes {
  emitterUpdateScript [
   &EmitterState(Life Cycle Mode=::NewEnumerator1, Loop Behavior=::NewEnumerator1, Loop Delay=0.03, Loop Duration=1, UseLoopDelay=true)
   &SpawnBurst_Instantaneous(Spawn Count=13, Use Spawn Probability=false, UseLoopCountLimit=false)
  ]
  particleSpawnScript [
   &InitializeParticle(
    Color=&HueShiftLinearColor(Hue=0, Saturation=1, Value=&RandomRangeFloat(Minimum=0.4, Maximum=0.8), Linear Color=$Particles.Color)
    Lifetime Max=0.3
    Lifetime Min=0.2
    Lifetime Mode=::NewEnumerator1
    Sprite Size Max=<30 40>
    Sprite Size Min=<10 12.5>
    Sprite Size Mode=::NewEnumerator2
    UsePositionOffset=false
    Write Material Random=true
   )
   &TorusLocation(
    Align Torus With Axis=false
    Coordinate Space:ENiagaraCoordinateSpace=::World
    Handle Radius=0
    Large Radius=0.16
    MaskSpawnGroup=false
    Radius Position=0
    Torus Distribution Mode:ENiagaraTorusDistributionMode=::NewEnumerator1
    U Position:NiagaraFloat=&RandomRangeFloat(Minimum=0, Maximum=1, Randomness Mode=::NewEnumerator2, Random Seed=?)
   )
   &AddVelocityFromPoint(
    Module.Use Falloff Distance=false
    Origin Offset=&RandomRangeVector(Minimum=<0 0 0>, Maximum=<0 0 -0.2>)
    Origin Offset Coordinate Space=::World
    Velocity Strength=&RandomRangeFloat(Minimum=32, Maximum=64)
   )
   &SubUVAnimation(End Frame=3, SubUV Animation Mode:ENiagara_SubUVLookupMode=::NewEnumerator2)
  ]
  particleUpdateScript [
   &ParticleState()
   &Drag(Drag=3, Use Linear Drag=true, Use Rotational Drag=false)
   &SolveForcesAndVelocity(Clamp Velocity=false, Limit Acceleration=false)
   &ScaleSpriteSize(
    Non-Uniform Curve Scale=&Multiply_Vector2D_ByFloat(Float=0.8, Vector2D=<1 1>)
    Non-Uniform Curve Sprite Scale=DataInterface() {propertyValues {
      XCurve {keys [{tangentMode=RCTM_User, time=0, value=1}, {tangentMode=RCTM_User, time=1, value=0.8}]}
      YCurve {keys [{tangentMode=RCTM_User, time=0, value=0.5}, {tangentMode=RCTM_User, time=1, value=1.7}]}
     }}
    Scale Sprite Size Mode=::NewEnumerator3
   )
   &ScaleColor(
    Scale Alpha=&Multiply_Float(
     A=&FloatFromCurve(
      FloatCurve=DataInterface() {propertyValues {Curve {keys [<at 0 0>,<at 0.1 1>,<at 0.2 0.8>,<at 1 0>]}}}
      Scale Curve=&FNTimeOfDayScalar(Night Scale=1, Morning Scale=0.3, Evening Scale=0.5)
     )
     B=0.1
    )
    Scale RGB=<1 0.8 0.8>
    ScaleA=true
    ScaleRGB=true
   )
   /SetVariables(Particles.Color2=&MultiplyLinearColor(Linear Color=$Particles.Color, Scale Factor=<1 1 1 0.2>)) {isSetParameters=true}
   &SpriteFacingAndAlignment(Facing Coordinate Space=::World, Sprite Facing=<0 0 1>, WriteAlignment=false, WriteFacing=true)
  ]
  rendererClasses [&NiagaraSpriteRendererProperties]
  renderers [
   {rendererClass=&NiagaraSpriteRendererProperties}
   {rendererClass=&NiagaraSpriteRendererProperties, rendererIndex=1}
  ]
 }
 Rocks_Burst {
  emitterUpdateScript [
   &EmitterState(Life Cycle Mode=::NewEnumerator1, Loop Behavior=::NewEnumerator1, Loop Duration=1, UseLoopDelay=true)
   &SpawnBurst_Instantaneous(
    Spawn Count=&RandomRangeInt(Minimum=7, Maximum=12)
    Use Spawn Probability=false
    UseLoopCountLimit=false
   )
   /SetVariables() {isSetParameters=true, module=SetVariables_#1}
  ]
  particleSpawnScript [
   &InitializeParticle(
    Lifetime Max=0.8
    Lifetime Min=0.6
    Lifetime Mode=::NewEnumerator1
    Mesh Scale Mode=::NewEnumerator3
    Mesh Uniform Scale=&Multiply_Float(
     A=&Lerp_Float(
      A=0.0285
      B=0.06
      Alpha=&FloatFromCurve(
       FloatCurve=DataInterface() {propertyValues {Curve {keys [<at 0 0 RCIM_Constant>, {arriveTangent=7.6187, interpMode=RCIM_Constant, leaveTangent=7.6187, time=0.65, value=1}]}}}
       CurveIndex=&RandomRangeFloat(Minimum=0, Maximum=1)
      )
     )
     B=1.4
    )
    UsePositionOffset=false
    Write Material Random=true
   )
   &ShapeLocation(
    Handle Radius=0
    Large Radius=2
    MaskSpawnGroup=false
    Offset=<0 0 7>
    Offset Mode=::NewEnumerator0
    Override Seed=false
    Shape Primitive=::NewEnumerator3
    UseOwnerScale=true
   )
   &InitialMeshOrientation(Use Rotation Vector=false)
   &AddRotationalVelocity(Rotation Rate=&Multiply_VectorByFloat(
     Vector=&RandomVector(Vector Scale=2)
     Float=&RandomRangeFloat(Minimum=0.5, Maximum=1)
    ))
   &DynamicMaterialParameters(
    Index 0 Param 2=1
    Module.Param0WriteEnabled=false
    Module.Param1WriteEnabled=true
    Module.Param2WriteEnabled=false
    Module.Param3WriteEnabled=false
   )
   &AddVelocity(
    Constrain To Radius=1000
    Module.Use Falloff Distance=false
    Origin Offset=<0 0 1>
    Override Seed=false
    Velocity Mode=::NewEnumerator1
    Velocity Speed=&RandomRangeFloat(Minimum=150, Maximum=250)
   )
   &AddVelocity(
    Cone Angle=90
    Cone Axis=<0 0 1>
    Inner Cone Angle=30
    Module.UseConeFalloff=false
    Module.UseVelDistribution=false
    Override Seed=false
    Velocity Mode=::NewEnumerator2
    Velocity Speed=&RandomRangeFloat(Minimum=50, Maximum=225)
   ) {module=AddVelocity001}
   &ApplyInitialForces()
   /SetVariables(Particles.MeshIndex=&RandomRangeInt(Minimum=0, Maximum=2)) {isSetParameters=true, module=SetVariables_#2}
  ]
  particleUpdateScript [
   &ParticleState()
   &GravityForce(Gravity=&RandomRangeVector(Minimum=<0 0 -750>, Maximum=<0 0 -1500>))
   &Collision(
    Age Colliding Particles=false
    KillOnCollision=false
    Module.Control Rest State Via Surface Normal=false
    Module.EnableMaxCollisionCount=false
    Module.Randomize Collision Normal Vector=false
    Restitution=0.2
   )
   &Drag(
    Drag=1.5
    Rotational Drag:NiagaraFloat=&FloatFromCurve(FloatCurve=DataInterface() {propertyValues {Curve {keys [<at 0 0 RCIM_Cubic>, {arriveTangent=0.1846, interpMode=RCIM_Cubic, leaveTangent=0.1846, tangentMode=RCTM_User, time=1, value=0.1}]}}})
    Use Linear Drag=true
    Use Rotational Drag=true
   )
   &SolveForcesAndVelocity(Clamp Velocity=false, Limit Acceleration=false)
   &ScaleMeshSize(Scale Factor=&VectorFromFloat(Value=&FloatFromCurve(FloatCurve=DataInterface() {propertyValues {Curve {keys [{arriveTangent=-0.0035, interpMode=RCIM_Cubic, leaveTangent=-0.0035, tangentMode=RCTM_User, time=0.6, value=1}, <at 1 0>]}}})))
  ]
  rendererClasses [&NiagaraMeshRendererProperties]
  renderers [{rendererClass=&NiagaraMeshRendererProperties, rendererIndex=0}]
 }
 Rocks_Burst001 {
  emitterUpdateScript [
   &EmitterState(Life Cycle Mode=::NewEnumerator1, Loop Behavior=::NewEnumerator1, Loop Duration=1, UseLoopDelay=true)
   &SpawnBurst_Instantaneous(
    Spawn Count=&RandomRangeInt(Minimum=7, Maximum=8)
    Use Spawn Probability=false
    UseLoopCountLimit=false
   )
   /SetVariables() {isSetParameters=true, module=SetVariables_#1}
  ]
  particleSpawnScript [
   &InitializeParticle(
    Lifetime Max=0.8
    Lifetime Min=0.6
    Lifetime Mode=::NewEnumerator1
    Mesh Scale Mode=::NewEnumerator3
    Mesh Uniform Scale=&Multiply_Float(
     A=&Lerp_Float(
      A=0.0285
      B=0.06
      Alpha=&FloatFromCurve(
       FloatCurve=DataInterface() {propertyValues {Curve {keys [<at 0 0 RCIM_Constant>, {arriveTangent=7.6187, interpMode=RCIM_Constant, leaveTangent=7.6187, time=0.65, value=1}]}}}
       CurveIndex=&RandomRangeFloat(Minimum=0, Maximum=1)
      )
     )
     B=1.4
    )
    UsePositionOffset=false
    Write Material Random=true
   )
   &ShapeLocation(
    Handle Radius=0
    Large Radius=2
    MaskSpawnGroup=false
    Offset=<0 0 10>
    Offset Mode=::NewEnumerator0
    Override Seed=false
    Shape Primitive=::NewEnumerator3
    UseOwnerScale=true
   )
   &InitialMeshOrientation(Use Rotation Vector=false)
   &AddRotationalVelocity(Rotation Rate=&Multiply_VectorByFloat(
     Vector=&RandomVector(Vector Scale=2)
     Float=&RandomRangeFloat(Minimum=0.5, Maximum=1)
    ))
   &DynamicMaterialParameters(
    Index 0 Param 2=1
    Module.Param0WriteEnabled=false
    Module.Param1WriteEnabled=true
    Module.Param2WriteEnabled=false
    Module.Param3WriteEnabled=false
   )
   &AddVelocity(Velocity=&RandomRangeVector(Minimum=<0 0 200>, Maximum=<0 0 325>))
   &AddVelocity(
    Cone Angle=120
    Cone Axis=&RotateVectorByQuaternion(
     Vector To Rotate=<1 0 0>
     Quaternion=&MakeQuatFromAxisAngle(Axis=<0 0 1>, AngleInDegrees=0, NormalizeAxis=true)
    )
    Inner Cone Angle=50
    Module.UseConeFalloff=false
    Module.UseVelDistribution=false
    Override Seed=false
    Velocity Mode=::NewEnumerator2
    Velocity Speed=&RandomRangeFloat(Minimum=70, Maximum=300)
   ) {module=AddVelocity001}
   &ApplyInitialForces()
   /SetVariables(Particles.MeshIndex=&RandomRangeInt(Minimum=0, Maximum=2)) {isSetParameters=true, module=SetVariables_#2}
  ]
  particleUpdateScript [
   &ParticleState()
   &GravityForce(Gravity=&RandomRangeVector(Minimum=<0 0 -750>, Maximum=<0 0 -1500>))
   &Collision(
    Age Colliding Particles=false
    KillOnCollision=false
    Module.Control Rest State Via Surface Normal=false
    Module.EnableMaxCollisionCount=false
    Module.Randomize Collision Normal Vector=false
    Restitution=0.2
   )
   &Drag(
    Drag=1.5
    Rotational Drag:NiagaraFloat=&FloatFromCurve(FloatCurve=DataInterface() {propertyValues {Curve {keys [<at 0 0 RCIM_Cubic>, {arriveTangent=0.1846, interpMode=RCIM_Cubic, leaveTangent=0.1846, tangentMode=RCTM_User, time=1, value=0.1}]}}})
    Use Linear Drag=true
    Use Rotational Drag=true
   )
   &SolveForcesAndVelocity(Clamp Velocity=false, Limit Acceleration=false)
   &ScaleMeshSize(Scale Factor=&VectorFromFloat(Value=&FloatFromCurve(FloatCurve=DataInterface() {propertyValues {Curve {keys [{arriveTangent=-0.0035, interpMode=RCIM_Cubic, leaveTangent=-0.0035, tangentMode=RCTM_User, time=0.6, value=1}, <at 1 0>]}}})))
  ]
  rendererClasses [&NiagaraMeshRendererProperties]
  renderers [{rendererClass=&NiagaraMeshRendererProperties, rendererIndex=0}]
 }
 Smoke_Center {
  emitterUpdateScript [
   &EmitterState(Life Cycle Mode=::NewEnumerator1, Loop Behavior=::NewEnumerator1, Loop Duration=1, UseLoopDelay=false)
   &SpawnBurst_Instantaneous(Spawn Count=5, Use Spawn Probability=false)
  ]
  particleSpawnScript [
   &InitializeParticle(
    Color=&RandomRangeLinearColor(Minimum=<rgba 0.401 0.401 0.401 1.3>, Maximum=<rgba 0.2135 0.2135 0.2135 1.3>)
    Lifetime Max=1
    Lifetime Min=0.5
    Lifetime Mode=::NewEnumerator1
    Sprite Rotation Mode=::NewEnumerator2
    Sprite Size Mode=::NewEnumerator1
    Uniform Sprite Size Max=4
    Uniform Sprite Size Min=3.2
    UsePositionOffset=false
    Write Material Random=true
   ) {module=InitializeParticle002}
   &TorusLocation(
    Align Torus With Axis=true
    Coordinate Space:ENiagaraCoordinateSpace=::World
    Large Radius=1.6
    MaskSpawnGroup=false
    Module.Override Randomness=false
    Torus Mode=::NewEnumerator2
   )
   &AddVelocityFromPoint(
    Module.Use Falloff Distance=false
    Origin Offset Coordinate Space=::World
    Velocity Strength=&RandomRangeFloat(Minimum=24, Maximum=160)
   )
   &AddVelocity(
    Cone Axis=&RotateVectorByQuaternion(
     Vector To Rotate=<1 0 0>
     Quaternion=&MakeQuatFromAxisAngle(Axis=<0 0 1>, AngleInDegrees=-145, NormalizeAxis=true)
    )
    Module.UseConeFalloff=false
    Module.UseVelDistribution=false
    Override Seed=false
    Velocity Mode=::NewEnumerator2
    Velocity Speed=&RandomRangeFloat(Minimum=0, Maximum=80)
   ) {module=AddVelocity001}
   &CameraOffset(Camera Offset Amount=50)
  ]
  particleUpdateScript [
   &ParticleState()
   &ScaleVelocity(
    Coordinate Space:ENiagaraCoordinateSpace=::World
    Velocity Scale=&Multiply_VectorByFloat(
     Vector=&RandomRangeVector(Minimum=<1 1 0>, Maximum=<1 1 0.5>)
     Float=&FloatFromCurve(
      FloatCurve=DataInterface() {propertyValues {Curve {keys [
          {arriveTangent=-6.3333, interpMode=RCIM_Cubic, leaveTangent=-6.3333, time=0, value=2}
          {arriveTangent=-6.3333, interpMode=RCIM_Cubic, leaveTangent=-6.3333, time=0.3, value=0.1}
         ]}}}
      Scale Curve=&RandomRangeFloat(Minimum=1, Maximum=1.2)
     )
    )
   )
   &Drag(Use Linear Drag=true, Use Rotational Drag=false)
   &SolveForcesAndVelocity(Clamp Velocity=false, Limit Acceleration=false)
   &ScaleSpriteSize(Scale Factor=&Vector2DFromFloat(Value=&FloatFromCurve(FloatCurve=DataInterface() {propertyValues {Curve {keys [
          {arriveTangent=12.5719, interpMode=RCIM_Cubic, leaveTangent=12.5719, tangentMode=RCTM_User, time=0, value=4}
          {interpMode=RCIM_Cubic, leaveTangent=-6.7553, time=1.0011, value=10}
         ]}}})))
   &ScaleColor(
    Scale Alpha=&FloatFromCurve(FloatCurve=DataInterface() {propertyValues {Curve {keys [
         {arriveTangent=-3.1757, interpMode=RCIM_Cubic, tangentMode=RCTM_Break, time=0, value=0.15}
         {interpMode=RCIM_Cubic, leaveTangent=-1.5121, tangentMode=RCTM_User, time=1.0011, value=0}
        ]}}})
    ScaleA=true
    ScaleRGB=true
   ) {module=ScaleColor001}
   &DynamicMaterialParameters(
    Index 0 Param 1=&Add_Float(
     A=&FloatFromCurve(FloatCurve=DataInterface() {propertyValues {Curve {keys [{arriveTangent=2.1549, interpMode=RCIM_Cubic, leaveTangent=2.1549, tangentMode=RCTM_User, time=0, value=0.5}, <at 1 1.25>]}}})
     B=&Multiply_Float(A=$Particles.Initial.MaterialRandom, B=0.5)
    )
    Index 0 Param 3=&FloatFromCurve(
     FloatCurve=DataInterface() {propertyValues {Curve {keys [
         {arriveTangent=-0.0944, interpMode=RCIM_Cubic, time=0, value=0}
         {arriveTangent=2.5569, interpMode=RCIM_Cubic, leaveTangent=2.5569, time=0.0519, value=2.3053}
         <at 0.5008 1.2804>
        ]}}}
     Scale Curve=15
    )
    Index 0 Param 4=1
    Index 1 Param 1=1
    Index 1 Param 2=1
    Index 1 Param 3=1
    Index 1 Param 4=1
    Index 2 Param 1=&FloatFromCurve(
     FloatCurve=DataInterface() {propertyValues {Curve {keys [
         {arriveTangent=0.8446, interpMode=RCIM_Cubic, leaveTangent=0.8446, tangentMode=RCTM_User, time=0, value=0.5}
         {arriveTangent=1.3293, interpMode=RCIM_Cubic, leaveTangent=1.3293, tangentMode=RCTM_User, time=1, value=2}
        ]}}}
     Scale Curve=0.25
    )
    Index 2 Param 2=&RandomRangeFloat(Minimum=0.5, Maximum=1)
    Index 2 Param 3=1
    Index 2 Param 4=1
    Module.Param0WriteEnabled=true
    Module.Param0WriteEnabled1=true
    Module.Param0WriteEnabled2=true
    Module.Param0WriteEnabled3=true
    Module.Param1WriteEnabled=true
    Module.Param1WriteEnabled1=true
    Module.Param1WriteEnabled2=true
    Module.Param1WriteEnabled3=true
    Module.Param2WriteEnabled=true
    Module.Param2WriteEnabled1=true
    Module.Param2WriteEnabled2=true
    Module.Param2WriteEnabled3=true
    Module.Param3WriteEnabled=true
    Module.Param3WriteEnabled1=true
    Module.Param3WriteEnabled2=true
    Module.Param3WriteEnabled3=true
    Write Parameter Index 1=true
    Write Parameter Index 2=true
    Write Parameter Index 3=true
   ) {module=DynamicMaterialParameters002}
  ]
  rendererClasses [&NiagaraSpriteRendererProperties]
  renderers [
   {rendererClass=&NiagaraSpriteRendererProperties}
   {rendererClass=&NiagaraSpriteRendererProperties, rendererIndex=1}
  ]
 }
 Smoke_Radial {
  emitterUpdateScript [
   &EmitterState()
   &SpawnBurst_Instantaneous(Spawn Count=7, Use Spawn Probability=false) {module=SpawnBurst_Instantaneous001}
  ]
  particleSpawnScript [
   &InitializeParticle(
    Color=<rgba 0.2656 0.2656 0.2656 0.2>
    Lifetime Max=1.3
    Lifetime Min=0.7
    Lifetime Mode=::NewEnumerator1
    Sprite Rotation Mode=::NewEnumerator2
    Sprite Size Mode=::NewEnumerator1
    Uniform Sprite Size Max=32
    Uniform Sprite Size Min=16
    UsePositionOffset=false
    Write Material Random=true
   ) {module=InitializeParticle001}
   &TorusLocation(
    Align Torus With Axis=true
    Coordinate Space:ENiagaraCoordinateSpace=::World
    Large Radius=2
    MaskSpawnGroup=false
    Module.Override Randomness=false
    Offset=<0 0 1>
    Torus Mode=::NewEnumerator2
   )
   &AddVelocityFromPoint(
    Module.Use Falloff Distance=false
    Origin Offset=<0 0 2>
    Origin Offset Coordinate Space=::World
    Velocity Origin:Vector3f=&SimulationPosition()
    Velocity Strength=&UniformRangedFloat(Minimum=64, Maximum=240)
   )
   &AddVelocity(
    Cone Axis=&RotateVectorByQuaternion(
     Vector To Rotate=<1 0 0>
     Quaternion=&MakeQuatFromAxisAngle(Axis=<0 0 1>, AngleInDegrees=-145, NormalizeAxis=true)
    )
    Module.UseConeFalloff=false
    Module.UseVelDistribution=false
    Override Seed=false
    Velocity Mode=::NewEnumerator2
    Velocity Speed=&RandomRangeFloat(Minimum=0, Maximum=64)
   ) {module=AddVelocity001}
   &ApplyInitialForces()
   &AddVelocity(Coordinate Space=::World, Velocity=&RandomRangeVector(Minimum=<0 0 0>, Maximum=<0 0 40>))
  ]
  particleUpdateScript [
   &ParticleState()
   &AccelerationForce(Acceleration=&MakeVector(X=0, Y=0, Z=&UniformRangedFloat(Minimum=0, Maximum=30)))
   &Drag(Drag=5, Use Linear Drag=true, Use Rotational Drag=false)
   &SolveForcesAndVelocity(Clamp Velocity=false, Limit Acceleration=false)
   &ScaleSpriteSize(Scale Factor=&Vector2DFromFloat(Value=&FloatFromCurve(FloatCurve=DataInterface() {propertyValues {Curve {keys [
          {arriveTangent=7.1593, interpMode=RCIM_Cubic, leaveTangent=7.1593, tangentMode=RCTM_User, time=0, value=0}
          {arriveTangent=1.1699, interpMode=RCIM_Cubic, leaveTangent=1.1699, tangentMode=RCTM_User, time=0.31, value=0.97}
          <at 1 1.25 RCIM_Cubic>
         ]}}})))
   &DynamicMaterialParameters(
    Index 0 Param 1=&FloatFromCurve(FloatCurve=DataInterface() {propertyValues {Curve {keys [
         {arriveTangent=0.9088, interpMode=RCIM_Cubic, leaveTangent=0.9088, tangentMode=RCTM_User, time=0, value=1}
         {arriveTangent=0.5473, interpMode=RCIM_Cubic, leaveTangent=0.5473, tangentMode=RCTM_User, time=1, value=2}
        ]}}})
    Index 0 Param 3=&FloatFromCurve(FloatCurve=DataInterface() {propertyValues {Curve {keys [
         {arriveTangent=-9.0447, interpMode=RCIM_Cubic, leaveTangent=-2.0044, tangentMode=RCTM_User, time=0, value=0.5}
         {interpMode=RCIM_Cubic, leaveTangent=4.6779, time=0.2851, value=0}
        ]}}})
    Index 1 Param 1=&FloatFromCurve(
     FloatCurve=DataInterface() {propertyValues {Curve {keys [<at 0 0>,<at 0.25 1>]}}}
     Scale Curve=&RandomRangeFloat(Minimum=2, Maximum=1)
    )
    Index 1 Param 2=&FloatFromCurve(
     FloatCurve=DataInterface() {propertyValues {Curve {keys [
         {interpMode=RCIM_Cubic, leaveTangent=3.7422, tangentMode=RCTM_Break, time=0, value=0.2}
         {interpMode=RCIM_Cubic, leaveTangent=0.9066, time=1, value=2}
        ]}}}
     Scale Curve=&RandomRangeFloat(Minimum=0.5, Maximum=1)
    )
    Index 1 Param 3=&FloatFromCurve(FloatCurve=DataInterface() {propertyValues {Curve {keys [
         {interpMode=RCIM_Cubic, leaveTangent=2.8445, tangentMode=RCTM_Break, time=0, value=0.2}
         {interpMode=RCIM_Cubic, leaveTangent=63.2978, time=0.25, value=1}
        ]}}})
    Index 1 Param 4=&FloatFromCurve(
     FloatCurve=DataInterface() {propertyValues {Curve {keys [<at 0 1>,<at 1 0.8>]}}}
     Scale Curve=&RandomRangeFloat(Minimum=0.5, Maximum=1)
    )
    Index 2 Param 1=&FloatFromCurve(FloatCurve=DataInterface() {propertyValues {Curve {keys [{interpMode=RCIM_Cubic, leaveTangent=3.2309, tangentMode=RCTM_Break, time=0.1, value=0}, <at 1 1>]}}})
    Index 2 Param 2=&FloatFromCurve(
     FloatCurve=DataInterface() {propertyValues {Curve {keys [<at 0 0.2 RCIM_Cubic>,<at 1 0.8 RCIM_Cubic>]}}}
     Scale Curve=&RandomRangeFloat(Minimum=0.2, Maximum=1.2)
    )
    Index 2 Param 3=&FloatFromCurve(FloatCurve=DataInterface() {propertyValues {Curve {keys [
         {interpMode=RCIM_Cubic, tangentMode=RCTM_Break, time=0, value=1}
         {interpMode=RCIM_Cubic, leaveTangent=11.1233, tangentMode=RCTM_Break, time=1, value=2}
        ]}}})
    Index 2 Param 4=&FloatFromCurve(
     FloatCurve=DataInterface() {propertyValues {Curve {keys [<at 0 1 RCIM_Cubic>, {interpMode=RCIM_Cubic, leaveTangent=-1.2956, tangentMode=RCTM_Break, time=1, value=0.75}]}}}
     Scale Curve=0.8
    )
    Module.Param0WriteEnabled=true
    Module.Param0WriteEnabled1=true
    Module.Param0WriteEnabled2=true
    Module.Param1WriteEnabled=true
    Module.Param1WriteEnabled1=true
    Module.Param1WriteEnabled2=true
    Module.Param2WriteEnabled=true
    Module.Param2WriteEnabled1=true
    Module.Param2WriteEnabled2=true
    Module.Param3WriteEnabled=true
    Module.Param3WriteEnabled1=true
    Module.Param3WriteEnabled2=true
    Write Parameter Index 1=true
    Write Parameter Index 2=true
   )
   &ScaleColor(
    Scale Alpha=&Multiply_Float(
     A=&FloatFromCurve(FloatCurve=DataInterface() {propertyValues {Curve {keys [<at 0 0>,<at 0.2049 0.9765>,<at 0.7791 0.9107>,<at 1 0>]}}})
     B=&Multiply_Float(A=&FNTimeOfDayScalar(Night Scale=1, Morning Scale=0.3, Evening Scale=0.3), B=1)
    )
    Scale RGB=&VectorFromFloat(Value=&FloatFromCurve(FloatCurve=DataInterface() {propertyValues {Curve {keys [<at 0 1>,<at 0.9983 0.5>]}}}))
    ScaleA=true
    ScaleRGB=true
   )
  ]
  rendererClasses [&NiagaraSpriteRendererProperties]
  renderers [{rendererClass=&NiagaraSpriteRendererProperties}]
 }
}
summary {
 systemName=NS_Emote_MedievalSheath_Impact
 emitters [
  {emitterName=Smoke_Center, rendererClasses [&NiagaraSpriteRendererProperties]}
  {emitterName=Smoke_Radial, rendererClasses [&NiagaraSpriteRendererProperties]}
  {emitterName=Impact_Spikes, rendererClasses [&NiagaraSpriteRendererProperties]}
  {emitterName=Rocks_Burst001, rendererClasses [&NiagaraMeshRendererProperties]}
  {emitterName=Rocks_Burst, rendererClasses [&NiagaraMeshRendererProperties]}
 ]
 userVariables []
}
